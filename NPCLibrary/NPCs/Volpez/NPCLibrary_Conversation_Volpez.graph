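# Conversation database for the Volpez NPC, providing greeting dialogue and access to quests and services.
# The entry point leads to one Speech node with a default line plus contextual lines for a player's first-ever conversation or specific worn outfits, prioritized through curve rows.
# The speech then branches through link nodes into the shared NPCLibrary quests and BasicServices conversation entries.
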ConversationDatabase /NPCLibrary/NPCs/Volpez/NPCLibrary_Conversation_Volpez
entry_points [{tag=Conversation.Entry.Volpez, destinations [#1]}]
nodes [
 ConversationEntryPointNode_7 {
  id=#1
  class=ConversationEntryPointNode
  outputs [#2]
  properties {entryTag {tagName=Conversation.Entry.Volpez}}
 }
 Speech_C_4 {
  id=#2
  class=Speech_C
  outputs [#3, #4]
  properties {
   nodeName=""
   generalConfig {
    defaultMessage=Let's make some tactical mischief.
    contextualMessages [
     {
      message="Do I spy a fellow trickster? Bring it in, buddy."
      contextRequirements [{requirement {class=FortControllerRequirement_IsFirstEverConversationWithNPC}}]
      priority {curve {rowName=Tandem.Speech.Priority.FirstEverInteraction}}
     }
     {
      message="No lectures today, big bro. I'm busy with the Tricksters."
      contextRequirements [{requirement {
         class=FortControllerRequirement_HasCID
         properties {allowedCIds [/BRCosmetics/Athena/Items/Cosmetics/Characters/CID_504_Athena_Commando_M_Lopex]}
        }}]
      priority {curve {rowName=Tandem.Speech.Priority.OutfitReaction}}
     }
     {
      message=You thought the Winter Tricksters were afraid of a little cold?
      contextRequirements [{requirement {
         class=FortControllerRequirement_HasCID
         properties {allowedCIds [/BRCosmetics/Athena/Items/Cosmetics/Characters/CID_983_Athena_Commando_F_CupidWinter]}
        }}]
      priority {curve {rowName=Tandem.Speech.Priority.OutfitReaction}}
     }
     {
      message=I like your swagger. How do you feel about mischief... and winter?
      contextRequirements [{requirement {class=FortControllerRequirement_HasCID, properties {allowedCIds [/BRCosmetics/Athena/Items/Cosmetics/Characters/Character_CrispRover]}}}]
      priority {curve {rowName=Tandem.Speech.Priority.OutfitReaction}}
     }
     {
      message=We took care of Snowheart and the icy red knight. Now you owe me a favour... or two.
      contextRequirements [{requirement {class=FortControllerRequirement_HasCID, properties {allowedCIds [&Character_Citadel]}}}]
      priority {curve {rowName=Tandem.Speech.Priority.OutfitReaction}}
     }
     {
      message="Ah, if it isn't my right hand bear! Let's talk tactics."
      contextRequirements [{requirement {
         class=FortControllerRequirement_HasCID
         properties {allowedCIds [/BRCosmetics/Athena/Items/Cosmetics/Characters/CID_640_Athena_Commando_M_TacticalBear]}
        }}]
      priority {curve {rowName=Tandem.Speech.Priority.OutfitReaction}}
     }
    ]
   }
  }
 }
 ConversationLinkNode_0 {id=#4, class=ConversationLinkNode, properties {remoteEntryTag {tagName=Conversation.Entry.BasicServices}}}
 ConversationLinkNode_1 {id=#3, class=ConversationLinkNode, properties {remoteEntryTag {tagName=Conversation.Entry.NPCLibrary.Quests}}}
]
